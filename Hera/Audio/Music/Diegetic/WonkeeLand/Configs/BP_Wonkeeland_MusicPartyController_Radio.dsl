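; Blueprint music party controller radio for WonkeeLand diegetic music that passes a firework music event to its audio component.
; On BeginPlay it waits for the current playlist data to be ready, then gets the event router from the game state.
; It listens for the GameplayEvent_FireworkMusic event and writes the payload's value to the bStartFireworks boolean parameter on its Audio component.
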(event EventBeginPlay
  (|CallParentFunction)
  (Class|FortAsyncActionCurrentPlaylistReady|PlaylistDataReadyAsync
    (:OnReady
      (bind _returnvalue (GameplayEventRouter|GetEventRouter "/Script/FortniteGame.FortGameState" self))
      (bind (_asynctaskproxy _payload) (GameplayEventRouter|StartListeningToEvent _returnvalue "/Hera/Audio/Music/Diegetic/WonkeeLand/Configs/GameplayEvent_FireworkMusic.GameplayEvent_FireworkMusic")
        (:OnEventReceived
          (Audio|Parameter|SetBooleanParameter (Variables|Default|GetAudio) "bStartFireworks" (Utilities|Struct|BreakGameplayEventFireworkMusic _payload) :!class /Script/Engine.AudioComponent))))))

(fn ConstructionScript
  (|CallParentFunction))
